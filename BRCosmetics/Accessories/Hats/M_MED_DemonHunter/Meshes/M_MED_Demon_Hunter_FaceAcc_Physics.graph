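# Physics asset for the Demon Hunter face accessory cosmetic, setting up physics for its left and right hoodie pull strings.
# Each string is a capsule chain from a base bone through three segments, joined by constraints with narrow swing and tighter twist limits.
# Both string bases and a small head capsule connect to a large spine_05 capsule through constraints with free swing and twist.
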PhysicsAsset /BRCosmetics/Accessories/Hats/M_MED_DemonHunter/Meshes/M_MED_Demon_Hunter_FaceAcc_Physics
bodies {
 dyn_hoodiePull_01_Base_l [{shapeType=Capsule, length=0.505, radius=0.505, rotation=<rot 0 -90 0>}]
 dyn_hoodiePull_01_Base_r [{shapeType=Capsule, length=0.5835, radius=0.544, center=<-0.2846 -0.1794 0.0858>, rotation=<rot 25.3486 -30.4005 -90>}]
 dyn_hoodiePull_01_l [{shapeType=Capsule, length=3.8969, radius=0.5382, center=<-1.8487 0 0.0099>, rotation=<rot 90 0 0>}]
 dyn_hoodiePull_01_r [{shapeType=Capsule, length=1.9345, radius=0.5393, center=<2.0031 0.1149 -0.0477>, rotation=<rot 0 89.9487 -85.9967>}]
 dyn_hoodiePull_02_l [{shapeType=Capsule, length=2.2131, radius=0.5235, center=<-1.1547 0 0>, rotation=<rot 90 0 0>}]
 dyn_hoodiePull_02_r [{shapeType=Capsule, length=0.8927, radius=0.4909, center=<0.8623 0.0145 0.0112>, rotation=<rot 0 90.7289 -86.8964>}]
 dyn_hoodiePull_03_l [{shapeType=Capsule, length=2.2131, radius=0.5235, center=<-1.1547 0 0>, rotation=<rot 90 0 0>}]
 dyn_hoodiePull_03_r [{shapeType=Capsule, length=0.9873, radius=0.4621, center=<0.9667 -0.018 -0.0161>, rotation=<rot 0 -91.2211 -88.4246>}]
 head [{shapeType=Capsule, length=1, radius=1}]
 spine_05 [{shapeType=Capsule, length=8.0253, radius=11.5041, center=<2.5758 0.8212 -0.1111>}]
}
constraints [
 {bone1Name=dyn_hoodiePull_01_l, bone2Name=dyn_hoodiePull_01_Base_l, swing1LimitDegrees=20, swing2LimitDegrees=20, twistLimitDegrees=5}
 {bone1Name=dyn_hoodiePull_02_l, bone2Name=dyn_hoodiePull_01_l, swing1LimitDegrees=20, swing2LimitDegrees=20, twistLimitDegrees=5}
 {bone1Name=dyn_hoodiePull_03_l, bone2Name=dyn_hoodiePull_02_l, swing1LimitDegrees=20, swing2LimitDegrees=20, twistLimitDegrees=5}
 {bone1Name=dyn_hoodiePull_01_r, bone2Name=dyn_hoodiePull_01_Base_r, swing1LimitDegrees=20, swing2LimitDegrees=20, twistLimitDegrees=5}
 {bone1Name=dyn_hoodiePull_02_r, bone2Name=dyn_hoodiePull_01_r, swing1LimitDegrees=20, swing2LimitDegrees=20, twistLimitDegrees=5}
 {bone1Name=dyn_hoodiePull_03_r, bone2Name=dyn_hoodiePull_02_r, swing1LimitDegrees=20, swing2LimitDegrees=20, twistLimitDegrees=5}
 {
  bone1Name=dyn_hoodiePull_01_Base_l
  bone2Name=spine_05
  swing1LimitDegrees=45
  swing1Motion=Free
  swing2LimitDegrees=45
  swing2Motion=Free
  twistLimitDegrees=45
  twistMotion=Free
 }
 {
  bone1Name=dyn_hoodiePull_01_Base_r
  bone2Name=spine_05
  swing1LimitDegrees=45
  swing1Motion=Free
  swing2LimitDegrees=45
  swing2Motion=Free
  twistLimitDegrees=45
  twistMotion=Free
 }
 {
  bone1Name=head
  bone2Name=spine_05
  swing1LimitDegrees=45
  swing1Motion=Free
  swing2LimitDegrees=45
  swing2Motion=Free
  twistLimitDegrees=45
  twistMotion=Free
 }
]
